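# ExponentialDensity is a library material function that maps a Depth input to exponential fog-style density falloff, modeled on D3DFOG_EXP.
# It multiplies Depth by Density, optionally squares that through the Use Exp2 static switch (on by default), and returns the reciprocal of e raised to it.
# An If outputs 1 for non-positive depth.
# Despite its description claiming inversion, the graph falls from 1 toward 0 as depth grows.
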MaterialFunction /Engine/Functions/Engine_MaterialFunctions01/Density/ExponentialDensity
graph {
 properties {
  bExposeToLibrary=true
  description="Uses the D3DFOG_EXP function to map exponential density. The result is inverted so that 0 is returned when depth = 0 and approaches 1 as depth increases."
 }
 nodes [
  n1 = FunctionInput() {bUsePreviewValueAsDefault=true, inputName=Depth, inputType=FunctionInput_Scalar}
  n2 = Constant() {r=1}
  n3 = FunctionInput() {bUsePreviewValueAsDefault=true, inputName=Density, inputType=FunctionInput_Scalar, sortPriority=1, previewValue=<0.66 0 0 1>}
  n4 = Multiply(A=@n1, B=@n3)
  n5 = FunctionInput(Preview=StaticBool() {value=true}) {
   bUsePreviewValueAsDefault=true
   description="Switch to the D3DFOG_EXP2 version (more contrast)"
   inputName=Use Exp2
   inputType=FunctionInput_StaticBool
   sortPriority=2
  }
  n6 = Power(Base=Constant() {desc=e, r=2.7183}, Exp=StaticSwitch(True=Multiply(A=@n4, B=@n4), False=@n4, Value=@n5) {desc=Exp2?})
  n7 = Constant() {r=1}
  FunctionOutput(None=If(A=@n1, B=Constant(), "A > B"=Divide(A=@n2, B=@n6), "A == B"=@n7, "A < B"=@n7))
 ]
}
